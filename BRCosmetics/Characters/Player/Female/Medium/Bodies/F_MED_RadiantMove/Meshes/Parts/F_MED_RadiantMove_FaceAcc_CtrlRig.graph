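# Control Rig asset for the face accessory part of the F_MED_RadiantMove female medium body, holding a bone hierarchy but no rig logic.
# The skeleton runs from root through spine, neck and head, with clavicle and upper-arm bones plus backpack and cape attach bones.
# Right, left and back hair chains branch from the head.
# Both the main graph and function library are empty, and no controls are defined.
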ControlRigBlueprint /BRCosmetics/Characters/Player/Female/Medium/Bodies/F_MED_RadiantMove/Meshes/Parts/F_MED_RadiantMove_FaceAcc_CtrlRig
graphs [
 RigVMFunctionLibrary {links [], nodes []}
 RigVMModel {links [], nodes []}
]
hierarchy [
 root {loc=<0 0 0>, rot=<0 0 0 1>}
 pelvis {parent=root, loc=<0 -0.5301 98.0571>, rot=<0.0127 0.707 -0.0127 -0.707>}
 spine_01 {parent=pelvis, loc=<6.1208 -0.9328 0>, rot=<0 0 0.034 -0.9994>}
 spine_02 {parent=spine_01, loc=<6.5997 -0.2095 0>, rot=<0 0 0.0238 -0.9997>}
 spine_03 {parent=spine_02, loc=<6.4914 -0.1237 0>, rot=<0 0 0.0149 -0.9999>}
 spine_04 {parent=spine_03, loc=<6.9598 0.0794 0>, rot=<0 0 -0.0132 -0.9999>}
 spine_05 {parent=spine_04, loc=<6.9987 0 0>, rot=<0 0 -0.0015 -1>}
 neck_01 {parent=spine_05, loc=<11.3138 0.1529 0>, rot=<0 0 -0.2003 -0.9797>}
 neck_02 {parent=neck_01, loc=<4.5627 0 0>, rot=<0 0 0.0504 -0.9987>}
 head {parent=neck_02, loc=<5.3914 0 0>, rot=<0 0 0.115 -0.9934>}
 Hair_01_r {parent=head, loc=<9.9883 5.7646 4.3713>, rot=<0.4127 0.3844 0.8026 0.1944>}
 Hair_02_r {parent=Hair_01_r, loc=<3.4669 0 0>, rot=<0 -0.4469 0 -0.8946>}
 Hair_03_r {parent=Hair_02_r, loc=<5.0759 0 0>, rot=<-0.0061 -0.1275 -0.0477 -0.9907>}
 Hair_04_r {parent=Hair_03_r, loc=<5.5968 0 0>, rot=<-0.0014 -0.0724 -0.0197 -0.9972>}
 Hair_05_r {parent=Hair_04_r, loc=<4.0091 0 0>, rot=<0 0 0 1>}
 TailHair_01_back {parent=head, loc=<11.7572 -6.4593 0>, rot=<0.5479 -0.447 0.447 -0.5479>}
 TailHair_02_back {parent=TailHair_01_back, loc=<5.1176 0 0>, rot=<0 -0.6301 0 -0.7765>}
 TailHair_03_back {parent=TailHair_02_back, loc=<4.2654 0 0>, rot=<0 -0.3131 0 -0.9497>}
 TailHair_04_back {parent=TailHair_03_back, loc=<5.2526 0 0>, rot=<0 0 0 1>}
 TailHair_05_back {parent=TailHair_04_back, loc=<3.8774 0 0>, rot=<0 0 0 1>}
 Hair_01_l {parent=head, loc=<10.0386 5.5414 -4.2562>, rot=<-0.0249 -0.9175 -0.3343 -0.2141>}
 Hair_02_l {parent=Hair_01_l, loc=<5.3664 0 0>, rot=<0 -0.2073 0 -0.9783>}
 Hair_03_l {parent=Hair_02_l, loc=<3.8155 0 0>, rot=<0 0 0 1>}
 clavicle_l {parent=spine_05, loc=<5.4831 -0.1053 -3.2946>, rot=<0.0465 -0.7289 -0.009 -0.6829>}
 upperarm_l {parent=clavicle_l, loc=<9.8658 -1.2136 -0.4105>, rot=<0.0164 -0.3245 -0.0269 -0.9454>}
 clavicle_r {parent=spine_05, loc=<5.4831 -0.1053 3.2946>, rot=<0.7289 0.0465 -0.6829 0.009>}
 upperarm_r {parent=clavicle_r, loc=<-9.8658 1.2136 0.4105>, rot=<0.0164 -0.3245 -0.0269 -0.9454>}
 attach_backpack {parent=spine_05, loc=<-0.123 -13.473 0>, rot=<-0.0093 -0.707 0.0093 -0.707>}
 attach_cape {parent=spine_05, loc=<9.017 -11.714 0>, rot=<0.0616 -0.7044 -0.0616 -0.7044>}
]
