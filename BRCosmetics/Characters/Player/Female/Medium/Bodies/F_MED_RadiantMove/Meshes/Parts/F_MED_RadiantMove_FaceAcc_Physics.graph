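# Physics asset for the F_MED_RadiantMove FaceAcc mesh part, set up for physics-driven hair on this female player cosmetic.
# Two five-bone capsule chains hang from the head: a right-side strand with very tight swing and twist limits, and a back tail with slightly looser limits.
# Capsules on the upper spine, clavicles, neck, head and right upper arm, joined by angularly locked constraints, give the hair a body to collide against.
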PhysicsAsset /BRCosmetics/Characters/Player/Female/Medium/Bodies/F_MED_RadiantMove/Meshes/Parts/F_MED_RadiantMove_FaceAcc_Physics
bodies {
 Hair_01_r [{shapeType=Capsule, length=1, radius=1}]
 Hair_02_r [{shapeType=Capsule, length=4.672, radius=3.0106, center=<1.9109 0 0>, rotation=<rot 0.0002 -0.0002 83.4897>}]
 Hair_03_r [{shapeType=Capsule, length=2.8013, radius=2.7598, center=<2.5053 0 0.173>, rotation=<rot -81.6743 -63.7738 151.6015>}]
 Hair_04_r [{shapeType=Capsule, length=2.4844, radius=2.4041, center=<0.5505 0 0.0035>, rotation=<rot -89.6377 0.002 -0.002>}]
 Hair_05_r [{shapeType=Capsule, length=3.9217, radius=1.9806, shapeName=Hair_05_r_capsule, center=<0.3846 -1 0>, rotation=<rot 0 158.2775 -76.5185>}]
 TailHair_01_back [{shapeType=Capsule, length=1, radius=1, center=<0.5741 0 0>}]
 TailHair_02_back [{shapeType=Capsule, length=3.418, radius=2.0647, center=<0.1405 0 0>}]
 TailHair_03_back [{shapeType=Capsule, length=5.5386, radius=4.7322, center=<0.1546 0 0>, rotation=<rot 0 0 90>}]
 TailHair_04_back [{shapeType=Capsule, length=6.86, radius=3.3129, center=<0.2617 0 0>, rotation=<rot -0.0002 0 -90.7799>}]
 TailHair_05_back [{shapeType=Capsule, length=3.9435, radius=2.5057, center=<1.0716 0 0.5>, rotation=<rot -62.0324 0 0>}]
 clavicle_l [{
   shapeType=Capsule
   length=8.0075
   radius=6.4191
   shapeName=clavicle_l_capsule
   center=<4.4519 0 -0.1972>
   rotation=<rot -0.0001 -90 -87.4642>
  }]
 clavicle_r [{shapeType=Capsule, length=8.0075, radius=6.4191, shapeName=clavicle_r_capsule, center=<-4.4519 0 0.1972>, rotation=<rot -0.0001 -90 92.5358>}]
 head [
  {
   shapeType=Capsule
   length=10.6187
   radius=7.0859
   shapeName=head_capsule
   center=<5.779 0.0033 0.2676>
   rotation=<rot 3.0344 94.2541 -88.9588>
  }
  {shapeType=Capsule, length=1.9973, radius=5.3663, shapeName=head_capsule, center=<3.4036 1.7868 2.2791>}
 ]
 neck_01 [{shapeType=Capsule, length=6.7571, radius=4.9147, shapeName=neck_01_capsule, rotation=<rot 0 -90 0>}]
 neck_02 [{shapeType=Capsule, length=4.2866, radius=5.3366, shapeName=neck_02_capsule, rotation=<rot 0 -90 0>}]
 spine_05 [{shapeType=Capsule, length=7.3016, radius=9.3932, shapeName=spine_05_capsule, center=<1.1266 0 0>, rotation=<rot 0 -90 0>}]
 upperarm_r [{
   shapeType=Capsule
   length=16.2459
   radius=4.2848
   shapeName=upperarm_r_capsule
   center=<-8.8455 0 0>
   rotation=<rot -0.0345 -89.9692 -91.1912>
  }]
}
constraints [
 {bone1Name=TailHair_02_back, bone2Name=TailHair_01_back, swing1LimitDegrees=10, swing2LimitDegrees=20, twistLimitDegrees=3}
 {bone1Name=TailHair_03_back, bone2Name=TailHair_02_back, swing1LimitDegrees=10, swing2LimitDegrees=20, twistLimitDegrees=3}
 {bone1Name=TailHair_04_back, bone2Name=TailHair_03_back, swing1LimitDegrees=10, swing2LimitDegrees=20, twistLimitDegrees=3}
 {bone1Name=TailHair_05_back, bone2Name=TailHair_04_back, swing1LimitDegrees=10, swing2LimitDegrees=20, twistLimitDegrees=3}
 {bone1Name=TailHair_01_back, bone2Name=head, swing1LimitDegrees=10, swing2LimitDegrees=20, twistLimitDegrees=3}
 {
  bone1Name=clavicle_r
  bone2Name=spine_05
  swing1LimitDegrees=45
  swing1Motion=Locked
  swing2LimitDegrees=45
  swing2Motion=Locked
  twistLimitDegrees=45
  twistMotion=Locked
 }
 {
  bone1Name=clavicle_l
  bone2Name=spine_05
  swing1LimitDegrees=45
  swing1Motion=Locked
  swing2LimitDegrees=45
  swing2Motion=Locked
  twistLimitDegrees=45
  twistMotion=Locked
 }
 {
  bone1Name=head
  bone2Name=neck_02
  swing1LimitDegrees=45
  swing1Motion=Locked
  swing2LimitDegrees=45
  swing2Motion=Locked
  twistLimitDegrees=45
  twistMotion=Locked
 }
 {
  bone1Name=neck_01
  bone2Name=spine_05
  swing1LimitDegrees=45
  swing1Motion=Locked
  swing2LimitDegrees=45
  swing2Motion=Locked
  twistLimitDegrees=45
  twistMotion=Locked
 }
 {
  bone1Name=neck_02
  bone2Name=neck_01
  swing1LimitDegrees=45
  swing1Motion=Locked
  swing2LimitDegrees=45
  swing2Motion=Locked
  twistLimitDegrees=45
  twistMotion=Locked
 }
 {bone1Name=Hair_05_r, bone2Name=Hair_04_r, swing1LimitDegrees=5, swing2LimitDegrees=5, twistLimitDegrees=2}
 {bone1Name=Hair_01_r, bone2Name=head, swing1LimitDegrees=5, swing2LimitDegrees=5, twistLimitDegrees=2}
 {bone1Name=Hair_02_r, bone2Name=Hair_01_r, swing1LimitDegrees=5, swing2LimitDegrees=5, twistLimitDegrees=2}
 {bone1Name=Hair_03_r, bone2Name=Hair_02_r, swing1LimitDegrees=5, swing2LimitDegrees=5, twistLimitDegrees=2}
 {bone1Name=Hair_04_r, bone2Name=Hair_03_r, swing1LimitDegrees=5, swing2LimitDegrees=5, twistLimitDegrees=2}
 {
  bone1Name=upperarm_r
  bone2Name=clavicle_r
  swing1LimitDegrees=45
  swing1Motion=Locked
  swing2LimitDegrees=45
  swing2Motion=Locked
  twistLimitDegrees=45
  twistMotion=Locked
 }
]
